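# Unlit alpha-composite material that ray marches an animated, noise-distorted SDF volume, heart-shaped by default, inside a cube's bounds.
# Local-space box intersection, optionally clipped by scene depth because depth test is off, sets ray start and step count, with blue-noise jitter and platform-scaled quality.
# Custom HLSL samples 2D pseudo-volume textures for density and baked shadow/AO, accumulating time-of-day light and ambient colours.
# Sphere and Gradient_SDF inputs are wired but unused.
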Material /Game/Global/Materials/MaterialGym/Materials/RayMarchedSDFVolume/M_VolRayMarchSDFNoise_OUT
graph {
 properties {
  bDisableDepthTest=true
  bUsedWithMeshDeformer=true
  bUsedWithNanite=true
  bUsedWithNiagaraMeshParticles=true
  bUsedWithSkeletalMesh=true
  bUsedWithStaticLighting=true
  blendMode=BLEND_AlphaComposite
  pixelDepthOffsetMode=PDOM_Legacy
  shadingModel=MSM_Unlit
  usedShadingModels=Unlit
 }
 nodes [
  MaterialExpressionTextureObjectParameter_6 = TextureObjectParameter() {
   desc=This texture is still at lifting aura folder
   parameterName=Gradient_SDF
   samplerType=SAMPLERTYPE_LinearColor
   texture=/Game/Animation/Props/Gadgets_LiftingAura/Textures/heartSDFGrad_v01_Volume
  }
  MaterialExpressionNamedRerouteUsage_53 = NamedRerouteUsage()
  MaterialExpressionNamedRerouteUsage_52 = NamedRerouteUsage()
  MaterialExpressionNamedRerouteUsage_51 = NamedRerouteUsage()
  MaterialExpressionNamedRerouteUsage_54 = NamedRerouteUsage()
  MaterialExpressionNamedRerouteUsage_55 = NamedRerouteUsage()
  n1 = TextureObjectParameter() {
   parameterName=Shadow
   samplerType=SAMPLERTYPE_LinearColor
   texture=/Game/Global/Materials/MaterialGym/Materials/RayMarchedSDFVolume/Textures/M_VolumeEncode_Global_Baked
  }
  MaterialExpressionReroute_4 = Reroute(None=@n1)
  MaterialExpressionReroute_5 = Reroute(None=ScalarParameter() {defaultValue=30, parameterName=densityGain})
  MaterialExpressionScalarParameter_24 = ScalarParameter() {defaultValue=0.4, parameterName=Shadow Power}
  MaterialExpressionMaterialFunctionCall_33 = MaterialFunctionCall(
   Color_Morning=VectorParameter() {parameterName=Ambient ColorM, defaultValue=<rgba 0.3298 0.123 0.4375 1>}.RGB
   Color_Day=VectorParameter() {parameterName=Ambient Color, defaultValue=<rgba 0.8385 0.1267 0.6782 1>}.RGB
   Color_Evening=VectorParameter() {parameterName=Ambient Color2, defaultValue=<rgba 0.2797 0.0947 0.1868 1>}.RGB
   Color_Night=VectorParameter() {parameterName=Ambient Color3, defaultValue=<rgba 0.0874 0.084 0.2708 1>}.RGB
  ) {materialFunction=&Color_Time_of_Day}
  MaterialExpressionReroute_7 = Reroute(None=MaterialFunctionCall(
    Color_Morning=VectorParameter() {parameterName=Light ColorM, defaultValue=<rgba 0.5873 0.112 0.5938 1>}.RGB
    Color_Day=VectorParameter() {parameterName=Light Color, defaultValue=<rgba 0.9062 0.1369 0.4969 1>}.RGB
    Color_Evening=VectorParameter() {parameterName=Light Color 2, defaultValue=<rgba 0.75 0.0952 0.2846 1>}.RGB
    Color_Night=VectorParameter() {parameterName=Light Color3, defaultValue=<rgba 0.3606 0.0848 0.4746 1>}.RGB
   ) {materialFunction=&Color_Time_of_Day}.Result)
  MaterialExpressionSkyAtmosphereLightDirection_2 = SkyAtmosphereLightDirection()
  MaterialExpressionTextureObjectParameter_5 = TextureObjectParameter() {parameterName=Noise, texture=/Game/Global/Materials/MaterialGym/Materials/RayMarchedSDFVolume/Textures/T_Volume_Global_PerlinWorley}
  MaterialExpressionTime_1 = Time()
  MaterialExpressionMultiply_38 = Multiply(A=@MaterialExpressionTime_1, B=ScalarParameter() {defaultValue=0.05, parameterName=NoiseTime}) {constB=0.015}
  MaterialExpressionScalarParameter_26 = ScalarParameter() {defaultValue=0.125, parameterName=NoiseStrength}
  MaterialExpressionScalarParameter_27 = ScalarParameter() {defaultValue=0.31, parameterName=SphereRadius}
  MaterialExpressionVectorParameter_17 = VectorParameter() {parameterName=SpherePosition, defaultValue=<rgba 0.5 0.2 0.475 1>}
  MaterialExpressionScalarParameter_28 = ScalarParameter() {
   defaultValue=0.666
   desc=0-1 Value. 1 is linear falloff from edge to sphere origin. 0 is hard edge sphere
   parameterName=NormalizedSphereFalloff
  }
  n2 = TextureObjectParameter() {
   parameterName=Stream
   samplerType=SAMPLERTYPE_LinearColor
   texture=/Game/Global/Materials/MaterialGym/Materials/RayMarchedSDFVolume/Textures/VT_Heart_Global_SDF_V01
  }
  MaterialExpressionReroute_8 = Reroute(None=@n2)
  MaterialExpressionScalarParameter_29 = ScalarParameter() {defaultValue=1, parameterName=AO Power}
  MaterialExpressionScalarParameter_30 = ScalarParameter() {defaultValue=0.4, parameterName=Inner Softness Distance}
  MaterialExpressionScalarParameter_31 = ScalarParameter() {defaultValue=0.8, parameterName=Density Darken Min}
  MaterialExpressionScalarParameter_32 = ScalarParameter() {defaultValue=2, parameterName=Density Darken Power}
  n3 = Custom!(
   VolumeTexture=@MaterialExpressionTextureObjectParameter_6
   rayStart=@MaterialExpressionNamedRerouteUsage_53
   numSteps=@MaterialExpressionNamedRerouteUsage_52
   cameraVector=@MaterialExpressionNamedRerouteUsage_51
   stepSize=@MaterialExpressionNamedRerouteUsage_54
   finalDelta=@MaterialExpressionNamedRerouteUsage_55
   ShadowTexture=@MaterialExpressionReroute_4
   densityGain=@MaterialExpressionReroute_5
   shadowGain=@MaterialExpressionScalarParameter_24
   ambientColor=@MaterialExpressionMaterialFunctionCall_33.Result
   lightColor=@MaterialExpressionReroute_7
   localLightDirection=@MaterialExpressionSkyAtmosphereLightDirection_2
   noise=@MaterialExpressionTextureObjectParameter_5
   timeAnim=@MaterialExpressionMultiply_38
   noiseStrength=@MaterialExpressionScalarParameter_26
   SphereRadius=@MaterialExpressionScalarParameter_27
   SpherePosition=@MaterialExpressionVectorParameter_17.RGB
   NormalizedSphereFalloff=@MaterialExpressionScalarParameter_28
   Streaming2DTex=@MaterialExpressionReroute_8
   AO_power=@MaterialExpressionScalarParameter_29
   innerSoftnessDistance=@MaterialExpressionScalarParameter_30
   DensityDarkenMin=@MaterialExpressionScalarParameter_31
   DensityDarkenPower=@MaterialExpressionScalarParameter_32
  ) {
   code="\r\n\r\nfloat theta = dot(normalize(localLightDirection), normalize(cameraVector));\r\nfloat3 cameraVectorOld = cameraVector;\r\ncameraVector *= stepSize;\r\n\r\n\r\nstruct Func {\r\n\r\n// the Henyey-Greenstein phase function\r\nfloat phase( float g,  float cos_theta)\r\n{\r\n    float denom = 1.0 + g * g - 2.0 * g * cos_theta;\r\n    return 1.0 / (4 * 3.14159) * (1.0 - g * g) / (denom * sqrt(denom));\r\n}\r\n\r\nfloat remapValueRange(float value, float low1, float high1, float low2,  float high2)\r\n{\r\n    return low2 + (value - low1) * (high2 - low2) / (high1 - low1);\r\n}\r\n\r\nfloat sphereDistance(float3 origin, float3 position, float radius){\r\n\r\n    //vector from position to sphere centre\r\n    float3 relativePosition = origin-position;\r\n\r\n    return length(relativePosition)-radius;\r\n}\r\n\r\n};\r\n\r\n\r\nfloat accumdist = 0;\r\nfloat current_density = 0;\r\nfloat transmittance = 1.0;\r\nfloat3 lightenergy = 0;\r\n\r\nFunc func;\r\n\r\nfloat height = GetPrimitiveData(Parameters.PrimitiveId).LocalObjectBoundsMax.z;\r\nfloat3 tim = (float3(timeAnim, timeAnim, timeAnim));\r\n\r\nfor (int i = 0; i < numSteps; i++){\r\n    //float4 noiseSample = noise.SampleLevel( noiseSampler, (rayStart + float3(0,-timeAnim*0.6, -timeAnim)/2),0);\r\n    float4 noiseSample = PseudoVolumeTexture( noise, noiseSampler, (rayStart + float3(0,-timeAnim*0.6, -timeAnim)/2), float2(16, 8), 128);\r\n    //float4 cursample = VolumeTexture.SampleLevel( VolumeTextureSampler, saturate(rayStart-(noiseSample.bga-0.5)*noiseStrength),0);\r\n    // float4 PseudoVolumeTexture(Texture2D Tex, SamplerState TexSampler, float3 inPos, float2 xysize, float numframes,\r\n\t//uint mipmode = 0, float miplevel = 0, float2 InDDX = 0, float2 InDDY = 0)\r\n    //float4 cursample = PseudoVolumeTexture( Streaming2DTex, Streaming2DTexSampler, saturate(rayStart-(noiseSample.bga-0.5)*noiseStrength), float2(128, 128), 121 );\r\n    float3 pos = saturate(rayStart-(noiseSample.bga-0.5)*noiseStrength);\r\n    float4 cursample = PseudoVolumeTexture( Streaming2DTex, Streaming2DTexSampler, pos, float2(11, 11), 121 );\r\n    float dens = func.remapValueRange(cursample.r, innerSoftnessDistance, 0.56, 1, 0);\r\n    float darken =  func.remapValueRange(saturate(pow(1.0-dens,DensityDarkenPower)), 0, 1,DensityDarkenMin, 1);\r\n    //dens *= saturate(func.remapValueRange(func.sphereDistance(SpherePosition, pos, SphereRadius), -SphereRadius*NormalizedSphereFalloff, 0.0, 0, 1));\r\n    //float2 curShadowsample = ShadowTexture.SampleLevel( VolumeTextureSampler, saturate(rayStart-(noiseSample.bga-0.5)*noiseStrength),0).rg;\r\n    float2 curShadowsample = PseudoVolumeTexture( ShadowTexture, ShadowTextureSampler, saturate(rayStart-(noiseSample.bga-0.5)*noiseStrength),float2(8, 8), 64).rg;\r\n    if(dens > 0.001){\r\n\r\n        current_density = saturate(dens)  * densityGain    *  stepSize;\r\n        float t_i = 1- exp(-current_density  );\r\n        \r\n        lightenergy += lightColor * t_i * transmittance * pow(curShadowsample.r, shadowGain) * darken;// * func.phase(0.1,theta)*1/stepSize ;\r\n        lightenergy += ambientColor * t_i * transmittance * pow(saturate(curShadowsample.g), AO_power) ;\r\n        //lightenergy += 5*pow(saturate(func.remapValueRange(func.sphereDistance(SpherePosition+float3(0,0.1,0), rayStart, SphereRadius), -SphereRadius*NormalizedSphereFalloff, 0.0, 0.1, 0)),2);\r\n        transmittance *= 1- t_i;  \r\n        \r\n        \r\n    }\r\n    rayStart += cameraVector;\r\n}\r\n\r\n\r\n// final step to remove artefact\r\n{\r\n    rayStart -= cameraVector;\r\n    rayStart += cameraVectorOld*finalDelta;\r\n\r\n    //float4 noiseSample = noise.SampleLevel( noiseSampler, (rayStart + float3(0,-timeAnim*0.6, -timeAnim)/2),0);\r\n    float4 noiseSample = PseudoVolumeTexture( noise, noiseSampler, (rayStart + float3(0,-timeAnim*0.6, -timeAnim)/2), float2(16, 8), 128);\r\n    //float4 cursample = VolumeTexture.SampleLevel( VolumeTextureSampler, saturate(rayStart-(noiseSample.bga-0.5)*noiseStrength),0);\r\n    float3 pos = saturate(rayStart-(noiseSample.bga-0.5)*noiseStrength);\r\n    float4 cursample = PseudoVolumeTexture( Streaming2DTex, Streaming2DTexSampler,  pos, float2(11, 11), 121 );\r\n    float dens = func.remapValueRange(cursample.a, innerSoftnessDistance, 0.56, 1, 0);\r\n    float darken =  func.remapValueRange(saturate(pow(1.0-dens,DensityDarkenPower)), 0, 1,DensityDarkenMin, 1);\r\n    //dens *= saturate(func.remapValueRange(func.sphereDistance(SpherePosition, pos, SphereRadius), -SphereRadius*NormalizedSphereFalloff, 0.0, 0, 1));\r\n    //float2 curShadowsample = ShadowTexture.SampleLevel( VolumeTextureSampler,saturate(rayStart-(noiseSample.bga-0.5)*noiseStrength),0).rg;\r\n    float2 curShadowsample = PseudoVolumeTexture( ShadowTexture, ShadowTextureSampler, saturate(rayStart-(noiseSample.bga-0.5)*noiseStrength),float2(8, 8), 64).rg;\r\n    if(dens > 0.001){\r\n\r\n        current_density = saturate(dens)  * densityGain    *  stepSize;\r\n        float t_i = 1- exp(-current_density  );\r\n        \r\n        lightenergy += lightColor * t_i * transmittance * pow(curShadowsample.r, shadowGain) * darken;// * func.phase(0.1,theta)*1/stepSize ;\r\n        lightenergy += ambientColor * t_i * transmittance *  pow(saturate(curShadowsample.g), AO_power);\r\n        transmittance *= 1- t_i;  \r\n        \r\n        \r\n    }\r\n}\r\n\r\nreturn float4( lightenergy, 1-transmittance);\r\n\r\n"
   description=DensityRayMarch
   outputType=CMOT_Float4
   showCode=true
   inputs [
    {
     inputName=VolumeTexture
     input {expression=&self:MaterialExpressionTextureObjectParameter_6}
    }
    {
     inputName=rayStart
     input {expression=&self:MaterialExpressionNamedRerouteUsage_53}
    }
    {
     inputName=numSteps
     input {expression=&self:MaterialExpressionNamedRerouteUsage_52}
    }
    {
     inputName=cameraVector
     input {expression=&self:MaterialExpressionNamedRerouteUsage_51}
    }
    {
     inputName=stepSize
     input {expression=&self:MaterialExpressionNamedRerouteUsage_54}
    }
    {
     inputName=finalDelta
     input {expression=&self:MaterialExpressionNamedRerouteUsage_55}
    }
    {
     inputName=ShadowTexture
     input {expression=&self:MaterialExpressionReroute_4}
    }
    {
     inputName=densityGain
     input {expression=&self:MaterialExpressionReroute_5}
    }
    {
     inputName=shadowGain
     input {expression=&self:MaterialExpressionScalarParameter_24}
    }
    {
     inputName=ambientColor
     input {expression=&self:MaterialExpressionMaterialFunctionCall_33}
    }
    {
     inputName=lightColor
     input {expression=&self:MaterialExpressionReroute_7}
    }
    {
     inputName=localLightDirection
     input {expression=&self:MaterialExpressionSkyAtmosphereLightDirection_2}
    }
    {
     inputName=noise
     input {expression=&self:MaterialExpressionTextureObjectParameter_5}
    }
    {
     inputName=timeAnim
     input {expression=&self:MaterialExpressionMultiply_38}
    }
    {
     inputName=noiseStrength
     input {expression=&self:MaterialExpressionScalarParameter_26}
    }
    {
     inputName=SphereRadius
     input {expression=&self:MaterialExpressionScalarParameter_27}
    }
    {
     inputName=SpherePosition
     input {
      expression=&self:MaterialExpressionVectorParameter_17
      mask=1
      maskB=1
      maskG=1
      maskR=1
     }
    }
    {
     inputName=NormalizedSphereFalloff
     input {expression=&self:MaterialExpressionScalarParameter_28}
    }
    {
     inputName=Streaming2DTex
     input {expression=&self:MaterialExpressionReroute_8}
    }
    {
     inputName=AO_power
     input {expression=&self:MaterialExpressionScalarParameter_29}
    }
    {
     inputName=innerSoftnessDistance
     input {expression=&self:MaterialExpressionScalarParameter_30}
    }
    {
     inputName=DensityDarkenMin
     input {expression=&self:MaterialExpressionScalarParameter_31}
    }
    {
     inputName=DensityDarkenPower
     input {expression=&self:MaterialExpressionScalarParameter_32}
    }
   ]
  }
  ComponentMask(None=@n3) {b=true, g=true, r=true}
  ComponentMask(None=@n3) {a=true}
  Custom!(None) {
   code="int3 randpos = int3(Parameters.SvPosition.xy, View.StateFrameIndexMod8);\r\nfloat rand =float(Rand3DPCG16(randpos).x) / 0xffff;\r\nreturn rand.x;"
   showCode=true
  }
  Custom() {
   code=" float scale = length( TransformLocalVectorToWorld(Parameters, float3(1.00000000,0.00000000,0.00000000)).xyz);\r\nfloat localscenedepth = CalcSceneDepth(ScreenAlignedPosition(GetScreenPosition(Parameters)));\r\n\r\nfloat3 camerafwd = mul(float3(0.00000000,0.00000000,1.00000000),ResolvedView.ViewToTranslatedWorld);\r\nlocalscenedepth /= (GetPrimitiveData(Parameters.PrimitiveId).LocalObjectBoundsMax.x * 2 * scale);\r\nlocalscenedepth /= abs( dot( camerafwd, Parameters.CameraVector ) );\r\nreturn localscenedepth;\r\n\r\n"
   description=camfwd
   outputType=CMOT_Float1
  }
  n4 = Custom!(in=@MaterialExpressionTime_1) {
   code="// The below expression will get compiled\r\n// into the output of this node\r\nreturn in;"
   outputType=CMOT_Float1
   showCode=true
   inputs [{
     inputName=in
     input {expression=&self:MaterialExpressionTime_1}
    }]
  }
  Frac(None=@n4)
  MaterialFunctionCall() {materialFunction=&ObjectLocalBounds}
  MaterialFunctionCall!(Steps, Plane alignment.) {materialFunction=/Volumetrics/Content/VolumeTextures/MaterialFunctions/VolumeBoxIntersect}
  n5 = ScalarParameter() {defaultValue=24, parameterName=maxSteps}
  n6 = Multiply(A=@n5, B=ScalarParameter() {defaultValue=0.5, parameterName=Mobile Quality Multipler})
  NamedRerouteDeclaration(None=MaterialFunctionCall("Xb1,PS4"=@n5, "PC-High&&Epic,XSX,PS5"=@n5, "PC-Perf,Android"=@n6, Switch=@n5, "PC-Low(NotPerf)&&Medium"=@n6) {materialFunction=&PlatformSwitch_1D}.Result) {name=MaxSteps, nodeColor=<rgba 0.8516 0 1 1>}
  n7 = MaterialFunctionCall!(Alpha Threshold, Random) {
   desc=This will take one texture sample and miplevel always 0
   materialFunction=/Game/TimeOfDay/CloudsAndStorm/Materials/MaterialFunctions/DitherTemporalAASpatioTemporalBlue
  }
  NamedRerouteDeclaration(None=@n7.Result) {name=BlueNoise, nodeColor=<rgba 1 0 0.7734 1>}
  n8 = CameraPositionWS()
  n9 = Constant()
  n10 = ScalarParameter() {parameterName=VerticalOffset}
  n11 = Divide(
   A=Add(
    A=TransformPosition(None=@n8) {transformSourceType=TRANSFORMPOSSOURCE_World, transformType=TRANSFORMPOSSOURCE_Instance}
    B=MaterialFunctionCall(X=@n9, Y=@n9, Z=@n10) {materialFunction=&MakeFloat3}.Result
   )
   B=ComponentMask(None=PreSkinnedLocalBounds().Max) {r=true}
  )
  n12 = Multiply(A=@n11) {constB=0.5, desc=The Local Space Pivot is at the center of the cube}
  NamedRerouteDeclaration(None=Add(A=@n12) {constB=0.5}) {name=Local Camera Position, desc=This value can go out of 1, nodeColor=<rgba 0.2891 0 1 1>}
  NamedRerouteDeclaration(None=MaterialFunctionCall(Vector 3=@n12) {materialFunction=/Engine/Functions/Engine_MaterialFunctions02/Utility/VectorLength}.V3 Length) {name=Camera Distance From Centre of Vol Box, nodeColor=<rgba 0.5 1 0 1>}
  NamedRerouteDeclaration(None=@n11) {name=Normalized Camera Position, nodeColor=<rgba 0 1 0.7891 1>}
  n13 = Transform(None=CameraVectorWS()) {transformSourceType=TRANSFORMSOURCE_World, transformType=TRANSFORM_Instance}
  NamedRerouteDeclaration(None=Multiply(A=Normalize(VectorInput=@n13)) {constB=-1}) {name=Local Camera Vector, nodeColor=<rgba 0 1 0.5781 1>}
  n14 = Divide(B=NamedRerouteUsage())
  n15 = Multiply(A=Subtract(A=Constant(), B=NamedRerouteUsage()), B=@n14)
  n16 = Multiply(A=Subtract(A=Constant() {r=1}, B=NamedRerouteUsage()), B=@n14)
  n17 = MaterialFunctionCall(Float3=Max(A=@n15, B=@n16)) {materialFunction=&BreakOutFloat3Components}
  n18 = Min(A=@n17.R, B=Min(A=@n17.G, B=@n17.G))
  n19 = ComponentMask(None=MaterialFunctionCall() {materialFunction=/Game/Animation/Props/Gadgets_LiftingAura/Materials/MaterialFunctions/MF_InstanceBounds}.Instance Bounds Max) {r=true}
  n20 = Transform(None=Constant3Vector() {constant=<rgba 1 0 0 0>}) {transformSourceType=TRANSFORMSOURCE_Instance}
  n21 = Divide(
   A=SceneDepth(UVs=MaterialFunctionCall() {materialFunction=&ScreenAlignedUVs}."X 100%, Y 100%").R
   B=Multiply(
    A=Multiply(A=@n19) {constB=2}
    B=MaterialFunctionCall(Vector 3=@n20) {materialFunction=&VectorLength}.V3 Length
   ) {constB=2}
  )
  n22 = Constant3Vector() {constant=<rgba 0 0 1 0>}
  n23 = Abs(None=DotProduct(
    A=CameraVectorWS()
    B=TransformPosition(None=@n22) {transformSourceType=TRANSFORMPOSSOURCE_View, transformType=TRANSFORMPOSSOURCE_TranslatedWorld}
   ))
  n24 = StaticSwitchParameter(True=Min(A=@n18, B=Divide(A=@n21, B=@n23)), False=@n18) {defaultValue=true, desc="Since material disables depth test, the sorting is done by scenedepth", parameterName=Composite Against Scene Depth}
  NamedRerouteDeclaration(None=@n24) {name=Far Distance t_1, nodeColor=<rgba 1 0 0.3984 1>}
  n25 = MaterialFunctionCall(Float3=Min(A=@n15, B=@n16)) {materialFunction=&BreakOutFloat3Components}
  n26 = Max(A=@n25.G, B=@n25.G)
  NamedRerouteDeclaration(None=Max(A=Max(A=@n25.R, B=@n26)) {constB=0}) {name=Near distance t_0, nodeColor=<rgba 0.875 0 1 1>}
  NamedRerouteDeclaration(None=Divide(A=Constant() {desc="sqrt(1^2+1^2+1^2)", r=1.7321}, B=NamedRerouteUsage())) {name=Step Size, desc=The size is relatived to cube, nodeColor=<rgba 1 0.9844 0 1>}
  n27 = Subtract(A=NamedRerouteUsage(), B=NamedRerouteUsage())
  n28 = Multiply(A=NamedRerouteUsage()) {constB=0.5}
  n29 = Subtract(A=Constant() {desc=Relative 1 unit, r=1}, B=Frac(None=Multiply(A=@n27, B=@n28)))
  n30 = NamedRerouteUsage()
  NamedRerouteDeclaration(None=Multiply(A=Multiply(A=@n29, B=@n30) {constB=0.2}, B=ScalarParameter() {defaultValue=1, parameterName=planeAlignment, sliderMax=1})) {name=Plane offset, nodeColor=<rgba 1 0 0.7734 1>}
  n31 = Multiply(A=NamedRerouteUsage(), B=NamedRerouteUsage())
  NamedRerouteDeclaration(None=@n31) {name=BlueStep, nodeColor=<rgba 0.027 0 1 1>}
  n32 = NamedRerouteUsage()
  n33 = Add(A=@n32, B=NamedRerouteUsage())
  n34 = MaterialFunctionCall("Xb1,PS4"=@n32, "PC-High&&Epic,XSX,PS5"=@n32, "PC-Perf,Android"=@n33, Switch=@n32, "PC-Low(NotPerf)&&Medium"=@n33) {materialFunction=&PlatformSwitch_1D}
  n35 = Add(A=NamedRerouteUsage(), B=Multiply(A=NamedRerouteUsage(), B=Max(A=@n34.Result) {constB=0})) {desc="The Raymarching here already skip the distance between camera and cube\r\nwhen camera position is over 0-1 space"}
  n36 = Add(A=@n35, B=Multiply(A=NamedRerouteUsage(), B=@n31))
  NamedRerouteDeclaration(None=MaterialFunctionCall!("Xb1,PS4"=@n36, "PC-High&&Epic,XSX,PS5"=@n36, "PC-Perf,Android"=@n35, Switch=@n36, "PC-Low(NotPerf)&&Medium"=@n35) {materialFunction=/Game/Effects/Niagara/MasterMaterials/Functions/Common/PlatformSwitch_3D}.Result) {name=Ray Start Position, nodeColor=<rgba 1 0 0.9141 1>}
  n37 = Max(A=Subtract(A=NamedRerouteUsage(), B=NamedRerouteUsage())) {constB=0}
  n38 = Subtract(A=@n37, B=NamedRerouteUsage())
  n39 = NamedRerouteUsage()
  n40 = Floor(None=Divide(
    A=MaterialFunctionCall("Xb1,PS4"=@n38, "PC-High&&Epic,XSX,PS5"=@n38, "PC-Perf,Android"=@n37, Switch=@n38, "PC-Low(NotPerf)&&Medium"=@n37) {materialFunction=&PlatformSwitch_3D}.Result
    B=@n39
   ))
  NamedRerouteDeclaration(None=@n40) {name=Step Count, nodeColor=<rgba 0 0.3906 1 1>}
  n41 = Multiply(A=@n39, B=@n40)
  NamedRerouteDeclaration(None=Subtract(A=NamedRerouteUsage(), B=Add(A=@n35, B=@n41))) {name=Final Delta Distance, nodeColor=<rgba 1 0.9375 0 1>}
  ViewProperty() {property=MEVP_TemporalSampleIndex}
 ]
}
